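# Physics asset for the Ice Maiden female face accessory, providing collision and joints for dynamic hair and a necklace.
# Box bodies form back, side and front hair chains hanging off the head with narrow twist limits, while a capsule necklace hangs off spine_05.
# Sphere and capsule bodies cover head, neck, clavicles and upper spine.
# The dyn_hair_fr_r_03 body constrains directly to the head, not to the preceding front-right segment.
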PhysicsAsset /BRCosmetics/Accessories/Hats/F_MED_Ice_Maiden_FaceAcc/Meshes/M_MED_Ice_Maiden_FaceAcc_Dyn
bodies {
 clavicle_l [{shapeType=Capsule, length=8.4761, radius=7.0211, center=<4.6329 -1.2864 -3.0895>, rotation=<rot 81.0288 0 180>}]
 clavicle_r [{shapeType=Capsule, length=8.4761, radius=7.0211, center=<-4.6032 1.2862 3.2778>, rotation=<rot 81.0295 0 180>}]
 dyn_hair_bk_01 [{shapeType=Box, extentX=8.4709, extentY=2.5821, extentZ=7.5, center=<4.2703 0 -0.0002>}]
 dyn_hair_bk_02 [{shapeType=Box, extentX=6.3411, extentY=1.8245, extentZ=7.5, center=<3.3967 0 -0.0001>}]
 dyn_hair_bk_03 [{shapeType=Box, extentX=5.291, extentY=1.8245, extentZ=7.5, center=<2.7336 -0.7629 -0.0757>, rotation=<rot 0 -19.0466 0>}]
 dyn_hair_bk_l_01 [{shapeType=Box, extentX=7.5, extentY=1.8245, extentZ=7.5, center=<3.1627 0 -0.0001>}]
 dyn_hair_bk_l_02 [{shapeType=Box, extentX=4.0292, extentY=1.8245, extentZ=7.5, center=<2.3261 0 -0.0002>}]
 dyn_hair_bk_l_03 [{shapeType=Box, extentX=1.5721, extentY=1.8245, extentZ=7.5, center=<0.9729 0.0001 0>}]
 dyn_hair_bk_r_01 [{shapeType=Box, extentX=7.5, extentY=1.8245, extentZ=7.5, center=<-3.1627 0 0>}]
 dyn_hair_bk_r_02 [{shapeType=Box, extentX=3.7083, extentY=1.8245, extentZ=7.5, center=<-2.4685 -0.0001 0>}]
 dyn_hair_bk_r_03 [
  {shapeType=Capsule, length=5.9962, radius=1, center=<-1.4384 -0.0004 0>}
  {shapeType=Box, extentX=1.0672, extentY=1.8245, extentZ=7.5, center=<-0.8276 0.0001 0>}
 ]
 dyn_hair_fr_l_01 [{shapeType=Box, extentX=3.1554, extentY=1.8245, extentZ=2.6624, center=<1.2327 0 -0.0001>}]
 dyn_hair_fr_l_02 [{shapeType=Box, extentX=3.1554, extentY=1.8245, extentZ=2.6624, center=<1.4171 0 -0.0001>}]
 dyn_hair_fr_r_01 [{shapeType=Box, extentX=2.5199, extentY=1.8245, extentZ=2.3115, center=<-1.3894 0 0.0004>}]
 dyn_hair_fr_r_02 [{shapeType=Box, extentX=2.5199, extentY=1.8245, extentZ=2.3115, center=<-1.3894 0 0.0004>}]
 dyn_hair_fr_r_03 [{shapeType=Box, extentX=2.5199, extentY=1.8245, extentZ=2.3115, center=<-1.3894 0 0.0004>}]
 dyn_hair_l_01 [{shapeType=Box, extentX=4.8098, extentY=1.8245, extentZ=7.5, center=<2.6918 0 -0.0001>}]
 dyn_hair_l_02 [{shapeType=Box, extentX=4.4187, extentY=1.8245, extentZ=7.5, center=<2.4511 0.0001 -0.0001>}]
 dyn_hair_l_03 [{shapeType=Box, extentX=1.8832, extentY=1.8245, extentZ=7.5, center=<1.2943 0 -0.0001>}]
 dyn_hair_r_01 [{shapeType=Box, extentX=4.8982, extentY=1.8245, extentZ=7.5, center=<-2.5648 0.0001 0>}]
 dyn_hair_r_02 [{shapeType=Box, extentX=4.3369, extentY=1.8245, extentZ=7.5, center=<-2.3181 0 0>}]
 dyn_hair_r_03 [{shapeType=Box, extentX=1.5836, extentY=1.8245, extentZ=7.5, center=<-1.0136 0 0>}]
 dyn_necklace [{shapeType=Capsule, length=3, radius=3, shapeName=dyn_necklace_capsule, center=<5.6 -1.5 0>, rotation=<rot 90 0 0>}]
 head [
  {shapeType=Sphere, radius=4.9352, center=<5.2907 -0.746 0>}
  {shapeType=Capsule, length=6.8946, radius=4.9643, center=<2.1415 1.5258 0>, rotation=<rot 90 0 0>}
 ]
 neck_01 [{shapeType=Capsule, length=8.4174, radius=5.2258, center=<3.3637 1.0519 -0.0001>, rotation=<rot 90 0 4.386>}]
 spine_05 [{shapeType=Capsule, length=10.6013, radius=8.8404, center=<0.4914 -1.9749 -0.0287>}]
}
constraints [
 {bone1Name=dyn_hair_bk_l_01, bone2Name=head, swing1LimitDegrees=45, swing2LimitDegrees=45, twistLimitDegrees=5}
 {bone1Name=dyn_hair_bk_l_02, bone2Name=dyn_hair_bk_l_01, swing1LimitDegrees=45, swing2LimitDegrees=45, twistLimitDegrees=5}
 {bone1Name=dyn_hair_bk_l_03, bone2Name=dyn_hair_bk_l_02, swing1LimitDegrees=45, swing2LimitDegrees=45, twistLimitDegrees=5}
 {bone1Name=dyn_hair_bk_01, bone2Name=head, swing1LimitDegrees=45, swing2LimitDegrees=45, twistLimitDegrees=5}
 {bone1Name=dyn_hair_bk_02, bone2Name=dyn_hair_bk_01, swing1LimitDegrees=45, swing2LimitDegrees=45, twistLimitDegrees=5}
 {bone1Name=dyn_hair_bk_03, bone2Name=dyn_hair_bk_02, swing1LimitDegrees=45, swing2LimitDegrees=45, twistLimitDegrees=5}
 {bone1Name=dyn_hair_l_01, bone2Name=head, swing1LimitDegrees=45, swing2LimitDegrees=45, twistLimitDegrees=5}
 {bone1Name=dyn_hair_l_02, bone2Name=dyn_hair_l_01, swing1LimitDegrees=45, swing2LimitDegrees=45, twistLimitDegrees=5}
 {bone1Name=dyn_hair_l_03, bone2Name=dyn_hair_l_02, swing1LimitDegrees=45, swing2LimitDegrees=45, twistLimitDegrees=5}
 {bone1Name=dyn_hair_fr_l_01, bone2Name=head, swing1LimitDegrees=45, swing2LimitDegrees=45, twistLimitDegrees=5}
 {bone1Name=dyn_hair_fr_l_02, bone2Name=dyn_hair_fr_l_01, swing1LimitDegrees=45, swing2LimitDegrees=45, twistLimitDegrees=5}
 {bone1Name=dyn_hair_bk_r_01, bone2Name=head, swing1LimitDegrees=45, swing2LimitDegrees=45, twistLimitDegrees=5}
 {bone1Name=dyn_hair_bk_r_02, bone2Name=dyn_hair_bk_r_01, swing1LimitDegrees=45, swing2LimitDegrees=45, twistLimitDegrees=5}
 {bone1Name=dyn_hair_bk_r_03, bone2Name=dyn_hair_bk_r_02, swing1LimitDegrees=45, swing2LimitDegrees=45, twistLimitDegrees=5}
 {bone1Name=dyn_hair_r_01, bone2Name=head, swing1LimitDegrees=45, swing2LimitDegrees=45, twistLimitDegrees=5}
 {bone1Name=dyn_hair_r_02, bone2Name=dyn_hair_r_01, swing1LimitDegrees=45, swing2LimitDegrees=45, twistLimitDegrees=5}
 {bone1Name=dyn_hair_r_03, bone2Name=dyn_hair_r_02, swing1LimitDegrees=45, swing2LimitDegrees=45, twistLimitDegrees=5}
 {bone1Name=dyn_hair_fr_r_01, bone2Name=head, swing1LimitDegrees=45, swing2LimitDegrees=45, twistLimitDegrees=5}
 {bone1Name=dyn_hair_fr_r_02, bone2Name=dyn_hair_fr_r_01, swing1LimitDegrees=45, swing2LimitDegrees=45, twistLimitDegrees=5}
 {bone1Name=head, bone2Name=neck_01, swing1LimitDegrees=45, swing2LimitDegrees=45, twistLimitDegrees=5}
 {
  bone1Name=clavicle_l
  bone2Name=spine_05
  swing1LimitDegrees=45
  swing1Motion=Free
  swing2LimitDegrees=45
  swing2Motion=Free
  twistLimitDegrees=45
  twistMotion=Free
 }
 {
  bone1Name=neck_01
  bone2Name=spine_05
  swing1LimitDegrees=45
  swing1Motion=Free
  swing2LimitDegrees=45
  swing2Motion=Free
  twistLimitDegrees=45
  twistMotion=Free
 }
 {
  bone1Name=clavicle_r
  bone2Name=spine_05
  swing1LimitDegrees=45
  swing1Motion=Free
  swing2LimitDegrees=45
  swing2Motion=Free
  twistLimitDegrees=45
  twistMotion=Free
 }
 {bone1Name=dyn_necklace, bone2Name=spine_05, swing1LimitDegrees=30, swing2LimitDegrees=20, twistLimitDegrees=5}
 {bone1Name=dyn_hair_fr_r_03, bone2Name=head, swing1LimitDegrees=45, swing2LimitDegrees=45, twistLimitDegrees=5}
]
